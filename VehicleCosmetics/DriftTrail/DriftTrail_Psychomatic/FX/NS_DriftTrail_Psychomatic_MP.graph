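# Niagara system for the Psychomatic drift trail vehicle cosmetic, placing effects at the two rear wheels.
# Two looping SubUV sprites and two scaled wheel-wrap mesh particles sample skeletal mesh bones, while two trail ribbons inherit source movement, offset partly by game mode.
# Painted-variant overrides color every emitter, with color-over-life curves on ribbons.
# Per-wheel FN_DisableTireFX masks each emitter, and per-side width attributes separate left and right ribbons.
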NiagaraSystem /VehicleCosmetics/DriftTrail/DriftTrail_Psychomatic/FX/NS_DriftTrail_Psychomatic_MP
dependencies {
 usedDataInterfaces [
  &NiagaraDataInterfaceSpriteRendererInfo
  &NiagaraDataInterfaceCurve
  &NiagaraDataInterfaceColorCurve
  &NiagaraDataInterfaceSkeletalMesh
 ]
 usedDynamicInputs []
 usedModules [
  &self:SystemSpawnScript.SystemScriptSource.SystemScriptGraph.NiagaraNodeAssignment_0.SetVariables_#1
  &FN_UnpackVehicleFlags
  /VehicleCosmetics/FX/Niagara/Modules/VC_SetVehicleSpeedAttribs
  &SystemState
  /Game/Effects/Niagara/Modules/Vehicle/FN_VehicleFlags_Cheatcodes
 ]
}
emitters {
 NE_VC_SubUV_Sprite {
  emitterUpdateScript [
   &EmitterState(
    Inactive Response=::NewEnumerator1
    Life Cycle Mode=::NewEnumerator1
    Loop Behavior=::NewEnumerator1
    Loop Duration Mode=::NewEnumerator1
    UseLoopDelay=false
   )
   &SpawnBurst_Instantaneous(Spawn Count=2, Use Spawn Probability=false, UseLoopCountLimit=false)
   &SpawnRate(SpawnRate=4, Use Spawn Probability=false) {enabled=false}
  ]
  particleSpawnScript [
   &VC_PaintedVariantsOverride(
    Black=<rgba 0.003 0.0024 0.0024 1>
    BurntSienna=<rgba 0.03 0.0053 0.003 1>
    Cobalt=<rgba 0.0176 0.0762 0.3515 1>
    Crimson=<rgba 0.55 0.0275 0.0283 1>
    DefaultColor=<rgba 1 0.01 0.0265 1>
    ForestGreen=<rgba 0.1311 0.4 0.08 1>
    Grey=<rgba 0.1272 0.1272 0.13 1>
    Lime=<rgba 0.3467 0.5029 0.0497 1>
    Orange=<rgba 0.45 0.0915 0.0098 1>
    Pink=<rgba 0.7991 0.0959 0.6249 1>
    Purple=<rgba 0.1652 0.0938 0.521 1>
    R1 Black:LinearColor=<rgba 0.0193 0.02 0.02 1>
    R1 BurntSienna:LinearColor=<rgba 0.25 0.0458 0.0125 1>
    R1 Cobalt:LinearColor=<rgba 0.3 0.9567 1 1>
    R1 Crimson:LinearColor=<rgba 1 0.0999 0.1046 1>
    R1 ForestGreen:LinearColor=<rgba 0.85 0.85 0.7979 1>
    R1 Grey:LinearColor=<rgba 0.6928 0.6928 0.7 1>
    R1 Lime:LinearColor=<rgba 1 1 0.3005 1>
    R1 Orange:LinearColor=<rgba 1 0.6435 0.07 1>
    R1 Pink:LinearColor=<rgba 1 0.7953 1 1>
    R1 Purple:LinearColor=<rgba 0.8 0.7305 0.8 1>
    R1 Saffron:LinearColor=<rgba 1 0.7907 0.3182 1>
    R1 SkyBlue:LinearColor=<rgba 1 1 1 1>
    R1 TitaniumWhite:LinearColor=<rgba 1 1 1 1>
    Saffron=<rgba 0.4508 0.3143 0.0496 1>
    SkyBlue=<rgba 0.1005 0.4014 0.5 1>
    TitaniumWhite=<rgba 0.1979 0.1979 0.2 1>
    Use R1 Painted Overrides=true
   )
   &InitializeParticle(
    Color Mode=::NewEnumerator0
    Lifetime=0.5
    Sprite Rotation Mode=::NewEnumerator2
    Sprite Size Mode=::NewEnumerator3
    Uniform Sprite Size=32
    UsePositionOffset=false
    Write Material Random=true
   )
   &PartitionParticles()
   &SkeletalMeshLocation(
    Bone / Socket Index=&Add_Integer(A=$Particles.PartitionParticles.Partition, B=4)
    Bone / Socket Sampling Mode=::NewEnumerator5
    Interpolate Mesh Sampling=false
    MaskSpawnGroup=false
    Mesh Sampling Type=::NewEnumerator6
    Sampled Position Offset=&AddVectorToPosition(Position=$User.offset, Vector=<-62 0 -10>)
    Skeletal Mesh=$System.SkeletalMesh
   )
   &AddVelocity(Velocity=&VC_Mask_Vectors(
     FrontEnd Value:Vector3f=<-1500 0 0>
     blsFrontEndPreview:NiagaraBool=false
     SpeedMin Value:Vector3f=<-800 0 0>
     SpeedMax Value:Vector3f=<-1 0 0>
     Speed Alpha:NiagaraFloat=$System.Speed_Alpha
    )) {enabled=false}
  ]
  particleUpdateScript [
   &FN_DisableTireFX(
    Mask Index Attribute:NiagaraInt32=$Particles.PartitionParticles.Partition
    Mask Index FL=2
    Mask Index FR=3
    Mask Index RL=0
    Mask Index RR=1
   )
   &ParticleState(Kill Particles When Lifetime Has Elapsed=false, Loop Particles Lifetime=true)
   &SolveForcesAndVelocity(Clamp Velocity=false, Limit Acceleration=false)
   &SubUVAnimation(
    Random Start Frame=true
    "Sprite Renderer "=DataInterface() {propertyValues {SpriteRenderer=&self:NE_VC_SubUV_Sprite_0.NiagaraSpriteRendererProperties_1}}
    UseEndFrame=false
    UseStartFrame=false
   )
   /SetVariables(Particles.OffNucleus=1) {isSetParameters=true}
  ]
  rendererClasses [&NiagaraSpriteRendererProperties]
  renderers [
   {rendererClass=&NiagaraSpriteRendererProperties}
   {rendererClass=&NiagaraSpriteRendererProperties, rendererIndex=1}
  ]
 }
 NE_VC_Trail_Ribbon {
  emitterUpdateScript [
   &EmitterState()
   &VC_RateQualityHandler(Epic Spawn Rate=45, Low Spawn Rate=45, Medium Spawn Rate=45, Use Custom Rate?=true)
   &VC_SpawnRibbon_v1(Number of Ribbons=2, Spawn Time=-1, SpawnRate=$Emitter.Spawn Rate)
  ]
  particleSpawnScript [
   &VC_PaintedVariantsOverride() {enabled=false}
   &VC_InitializeRibbon_v1(
    Color=$Particles.Color
    Custom Offset=true
    Enable Skeletal Position=true
    Lifetime=0.9
    Local Vector Offset=&AddVectorToPosition(
     Position=&MergeHelper_VectorSwitch(0=<-40 0 -10>, 1=<-44 0 -6>, 2=<-44 0 -6>, 3=?, Index=$User.GameModeID, Num=::NewEnumerator1)
     Vector=Linked() {linkedVariable {name=User.offset, type=NiagaraPosition}}
    )
    Offset Amount=1
    Ribbon Width=90
   )
   &AddVelocity(Velocity=&VC_Mask_Vectors(
     FrontEnd Value:Vector3f=<-1000 0 0>
     blsFrontEndPreview:NiagaraBool=true
     SpeedMin Value:Vector3f=<-1000 0 0>
     SpeedMax Value:Vector3f=<-1000 0 0>
     Speed Alpha:NiagaraFloat=$System.Speed_Alpha
    )) {module=AddVelocity001}
  ]
  particleUpdateScript [
   &FN_DisableTireFX(Mask Index Attribute:NiagaraInt32=$Particles.SpawnGroup, Mask Index FL=2, Mask Index FR=3, Mask Index RL=0, Mask Index RR=1)
   /SetVariables(
    Particles.RL_Width:NiagaraFloat=&SelectFloatFromArray(Direct Array Index=$Particles.SpawnGroup, Float Selection Array=DataInterface() {propertyValues {FloatData [90,0]}})
    Particles.RR_Width:NiagaraFloat=&SelectFloatFromArray(Direct Array Index=$Particles.SpawnGroup, Float Selection Array=DataInterface() {propertyValues {FloatData [0,90]}})
   ) {isSetParameters=true}
   &ParticleState()
   &InheritSourceMovement(
    Applied Position Scale=&Multiply_VectorByFloat(
     Vector=<1 1 1>
     Float=&FloatFromCurve(FloatCurve=DataInterface() {propertyValues {Curve {keys [<at 0.3 1 RCIM_Cubic>, {arriveTangent=-3.0131, interpMode=RCIM_Cubic, leaveTangent=-3.0131, tangentMode=RCTM_User, time=1, value=0}]}}})
    )
    ApplyForce=false
    ApplyPosition=true
    ApplyPositionFalloff=false
   )
   &SolveForcesAndVelocity(Clamp Velocity=false, Limit Acceleration=false)
   &VC_UpdateRibbon_v1(Enable Skeletal Position=true)
   &VC_PaintedVariantsOverride(
    Black=&ColorFromCurve(ColorCurve=DataInterface() {propertyValues {
       AlphaCurve {keys [<at 0 2 RCIM_Cubic>,<at 1 2 RCIM_Cubic>]}
       BlueCurve {keys [<at 0.1814 0.003 RCIM_Cubic>,<at 0.6008 0.0021 RCIM_Cubic>]}
       GreenCurve {keys [<at 0.1814 0.0028 RCIM_Cubic>,<at 0.6008 0.0021 RCIM_Cubic>]}
       RedCurve {keys [<at 0.1814 0.0028 RCIM_Cubic>,<at 0.6008 0.0021 RCIM_Cubic>]}
      }})
    BurntSienna=&ColorFromCurve(ColorCurve=DataInterface() {propertyValues {
       AlphaCurve {keys [<at 0 1 RCIM_Cubic>,<at 1 1.5 RCIM_Cubic>]}
       BlueCurve {keys [<at 0.1814 0.0018 RCIM_Cubic>,<at 0.6008 0.0012 RCIM_Cubic>]}
       GreenCurve {keys [<at 0.1814 0.0056 RCIM_Cubic>,<at 0.6008 0.003 RCIM_Cubic>]}
       RedCurve {keys [<at 0.1814 0.0467 RCIM_Cubic>,<at 0.6008 0.0185 RCIM_Cubic>]}
      }})
    Cobalt=&ColorFromCurve(ColorCurve=DataInterface() {propertyValues {
       AlphaCurve {keys [<at 0 2>,<at 1 2 RCIM_Cubic>]}
       BlueCurve {keys [<at 0.1814 0.6308 RCIM_Cubic>,<at 0.6008 0.2918 RCIM_Cubic>]}
       GreenCurve {keys [<at 0.1814 0.092 RCIM_Cubic>,<at 0.6008 0.0319 RCIM_Cubic>]}
       RedCurve {keys [<at 0.1814 0 RCIM_Cubic>,<at 0.6008 0.008 RCIM_Cubic>]}
      }})
    Crimson=&ColorFromCurve(ColorCurve=DataInterface() {propertyValues {
       AlphaCurve {keys [<at 0 1 RCIM_Cubic>,<at 1 1 RCIM_Cubic>]}
       BlueCurve {keys [<at 0.1814 0.0225 RCIM_Cubic>,<at 0.6008 0.0056 RCIM_Cubic>]}
       GreenCurve {keys [<at 0.1814 0.0216 RCIM_Cubic>,<at 0.6008 0.0052 RCIM_Cubic>]}
       RedCurve {keys [<at 0.1814 0.6 RCIM_Cubic>,<at 0.6008 0.2016 RCIM_Cubic>]}
      }})
    DefaultColor=&ColorFromCurve(ColorCurve=DataInterface() {propertyValues {
       AlphaCurve {keys [<at 0 1 RCIM_Cubic>,<at 1 1 RCIM_Cubic>]}
       BlueCurve {keys [<at 0.1814 0 RCIM_Cubic>,<at 0.6008 0 RCIM_Cubic>]}
       GreenCurve {keys [{arriveTangent=1.4563, interpMode=RCIM_Cubic, time=0.1814, value=0.0214}, <at 0.6008 0.8963 RCIM_Cubic>]}
       RedCurve {keys [{arriveTangent=-1.1203, interpMode=RCIM_Cubic, time=0.1814, value=0.9}, <at 0.6008 0.227 RCIM_Cubic>]}
      }})
    ForestGreen=&ColorFromCurve(ColorCurve=DataInterface() {propertyValues {
       AlphaCurve {keys [<at 0 1 RCIM_Cubic>,<at 1 1 RCIM_Cubic>]}
       BlueCurve {keys [<at 0.1814 0.0395 RCIM_Cubic>,<at 0.6008 0.0256 RCIM_Cubic>]}
       GreenCurve {keys [<at 0.1814 0.4 RCIM_Cubic>,<at 0.6008 0.15 RCIM_Cubic>]}
       RedCurve {keys [<at 0.1814 0.0972 RCIM_Cubic>,<at 0.6008 0.0404 RCIM_Cubic>]}
      }})
    Grey=&ColorFromCurve(ColorCurve=DataInterface() {propertyValues {
       AlphaCurve {keys [<at 0 1 RCIM_Cubic>,<at 1 2 RCIM_Cubic>]}
       BlueCurve {keys [<at 0.1814 0.1413 RCIM_Cubic>,<at 0.6008 0.05 RCIM_Cubic>]}
       GreenCurve {keys [<at 0.1814 0.1413 RCIM_Cubic>,<at 0.6008 0.0482 RCIM_Cubic>]}
       RedCurve {keys [<at 0.1814 0.1413 RCIM_Cubic>,<at 0.6008 0.0482 RCIM_Cubic>]}
      }})
    Lime=&ColorFromCurve(ColorCurve=DataInterface() {propertyValues {
       AlphaCurve {keys [<at 0 1 RCIM_Cubic>,<at 1 1 RCIM_Cubic>]}
       BlueCurve {keys [<at 0.1814 0 RCIM_Cubic>,<at 0.6008 0.0097 RCIM_Cubic>]}
       GreenCurve {keys [<at 0.1814 0.5029 RCIM_Cubic>,<at 0.6008 0.0999 RCIM_Cubic>]}
       RedCurve {keys [<at 0.1814 0.3296 RCIM_Cubic>,<at 0.6008 0.0612 RCIM_Cubic>]}
      }})
    Orange=&ColorFromCurve(ColorCurve=DataInterface() {propertyValues {
       AlphaCurve {keys [<at 0 1 RCIM_Cubic>,<at 1 1 RCIM_Cubic>]}
       BlueCurve {keys [<at 0.1814 0.011 RCIM_Cubic>,<at 0.6008 0.0052 RCIM_Cubic>]}
       GreenCurve {keys [<at 0.1814 0.1022 RCIM_Cubic>,<at 0.6008 0.0343 RCIM_Cubic>]}
       RedCurve {keys [<at 0.1814 0.5029 RCIM_Cubic>,<at 0.6008 0.2502 RCIM_Cubic>]}
      }})
    Pink=&ColorFromCurve(ColorCurve=DataInterface() {propertyValues {
       AlphaCurve {keys [<at 0 1 RCIM_Cubic>,<at 1 1 RCIM_Cubic>]}
       BlueCurve {keys [<at 0.1814 0.82 RCIM_Cubic>,<at 0.6008 0.117 RCIM_Cubic>]}
       GreenCurve {keys [<at 0.1814 0.0947 RCIM_Cubic>,<at 0.6008 0.0168 RCIM_Cubic>]}
       RedCurve {keys [<at 0.1814 0.9473 RCIM_Cubic>,<at 0.6008 0.1714 RCIM_Cubic>]}
      }})
    Purple=&ColorFromCurve(ColorCurve=DataInterface() {propertyValues {
       AlphaCurve {keys [<at 0 1 RCIM_Cubic>,<at 1 1 RCIM_Cubic>]}
       BlueCurve {keys [<at 0.1814 0.521 RCIM_Cubic>,<at 0.6008 0.25 RCIM_Cubic>]}
       GreenCurve {keys [<at 0.1814 0.1042 RCIM_Cubic>,<at 0.6008 0.025 RCIM_Cubic>]}
       RedCurve {keys [<at 0.1814 0.1739 RCIM_Cubic>,<at 0.6008 0.0438 RCIM_Cubic>]}
      }})
    R1 Black:LinearColor=&ColorFromCurve(ColorCurve=DataInterface() {propertyValues {
       AlphaCurve {keys [<at 0 1 RCIM_Cubic>,<at 1 1 RCIM_Cubic>]}
       BlueCurve {keys [<at 0.1871 0.0296 RCIM_Cubic>,<at 0.6036 0.0091 RCIM_Cubic>]}
       GreenCurve {keys [<at 0.1871 0.0284 RCIM_Cubic>,<at 0.6036 0.0091 RCIM_Cubic>]}
       RedCurve {keys [<at 0.1871 0.0284 RCIM_Cubic>,<at 0.6036 0.0091 RCIM_Cubic>]}
      }})
    R1 BurntSienna:LinearColor=&ColorFromCurve(ColorCurve=DataInterface() {propertyValues {
       AlphaCurve {keys [<at 0 1 RCIM_Cubic>,<at 1 1 RCIM_Cubic>]}
       BlueCurve {keys [<at 0.1871 0.006 RCIM_Cubic>,<at 0.6036 0.0044 RCIM_Cubic>]}
       GreenCurve {keys [<at 0.1871 0.0513 RCIM_Cubic>,<at 0.6036 0.0284 RCIM_Cubic>]}
       RedCurve {keys [<at 0.1871 0.3515 RCIM_Cubic>,<at 0.6036 0.2232 RCIM_Cubic>]}
      }})
    R1 Cobalt:LinearColor=&ColorFromCurve(ColorCurve=DataInterface() {propertyValues {
       AlphaCurve {keys [<at 0 1 RCIM_Cubic>,<at 1 1 RCIM_Cubic>]}
       BlueCurve {keys [<at 0.1871 1 RCIM_Cubic>,<at 0.6036 1 RCIM_Cubic>]}
       GreenCurve {keys [<at 0.1871 0.7835 RCIM_Cubic>,<at 0.6036 0.3968 RCIM_Cubic>]}
       RedCurve {keys [<at 0.1871 0.3005 RCIM_Cubic>,<at 0.6036 0.1095 RCIM_Cubic>]}
      }})
    R1 Crimson:LinearColor=&ColorFromCurve(ColorCurve=DataInterface() {propertyValues {
       AlphaCurve {keys [<at 0 1 RCIM_Cubic>,<at 1 1 RCIM_Cubic>]}
       BlueCurve {keys [<at 0.1871 0.1046 RCIM_Cubic>,<at 0.6036 0.0305 RCIM_Cubic>]}
       GreenCurve {keys [<at 0.1871 0.0999 RCIM_Cubic>,<at 0.6036 0.0305 RCIM_Cubic>]}
       RedCurve {keys [<at 0.1871 1 RCIM_Cubic>,<at 0.6036 0.4945 RCIM_Cubic>]}
      }})
    R1 DefaultColor=&ColorFromCurve(ColorCurve=DataInterface() {propertyValues {
       AlphaCurve {keys [<at 0 1 RCIM_Cubic>,<at 1 1 RCIM_Cubic>]}
       BlueCurve {keys [<at 0.1871 0.4801 RCIM_Cubic>,<at 0.6036 0.4801 RCIM_Cubic>]}
       GreenCurve {keys [<at 0.1871 0.59 RCIM_Cubic>,<at 0.6036 1 RCIM_Cubic>]}
       RedCurve {keys [<at 0.1871 1 RCIM_Cubic>,<at 0.6036 0.6032 RCIM_Cubic>]}
      }})
    R1 ForestGreen:LinearColor=&ColorFromCurve(ColorCurve=DataInterface() {propertyValues {
       AlphaCurve {keys [<at 0 1 RCIM_Cubic>,<at 1 1 RCIM_Cubic>]}
       BlueCurve {keys [<at 0.1871 0.2572 RCIM_Cubic>,<at 0.6036 0.08 RCIM_Cubic>]}
       GreenCurve {keys [<at 0.1871 0.6 RCIM_Cubic>,<at 0.6036 0.4 RCIM_Cubic>]}
       RedCurve {keys [<at 0.1871 0.3908 RCIM_Cubic>,<at 0.6036 0.1393 RCIM_Cubic>]}
      }})
    R1 Grey:LinearColor=&ColorFromCurve(ColorCurve=DataInterface() {propertyValues {
       AlphaCurve {keys [<at 0 1 RCIM_Cubic>,<at 1 1 RCIM_Cubic>]}
       BlueCurve {keys [<at 0.1871 0.65 RCIM_Cubic>,<at 0.6036 0.15 RCIM_Cubic>]}
       GreenCurve {keys [<at 0.1871 0.65 RCIM_Cubic>,<at 0.6036 0.15 RCIM_Cubic>]}
       RedCurve {keys [<at 0.1871 0.65 RCIM_Cubic>,<at 0.6036 0.15 RCIM_Cubic>]}
      }})
    R1 Lime:LinearColor=&ColorFromCurve(ColorCurve=DataInterface() {propertyValues {
       AlphaCurve {keys [<at 0 1 RCIM_Cubic>,<at 1 1 RCIM_Cubic>]}
       BlueCurve {keys [<at 0.1871 0.2 RCIM_Cubic>,<at 0.6036 0.16 RCIM_Cubic>]}
       GreenCurve {keys [<at 0.1871 1 RCIM_Cubic>,<at 0.6036 0.8 RCIM_Cubic>]}
       RedCurve {keys [<at 0.1871 1 RCIM_Cubic>,<at 0.6036 0.8 RCIM_Cubic>]}
      }})
    R1 Orange:LinearColor=&ColorFromCurve(ColorCurve=DataInterface() {propertyValues {
       AlphaCurve {keys [<at 0 1 RCIM_Cubic>,<at 1 1 RCIM_Cubic>]}
       BlueCurve {keys [<at 0.1871 0.0699 RCIM_Cubic>,<at 0.5975 0.016 RCIM_Cubic>]}
       GreenCurve {keys [<at 0.1871 0.6212 RCIM_Cubic>,<at 0.5975 0.1444 RCIM_Cubic>]}
       RedCurve {keys [<at 0.1871 0.6997 RCIM_Cubic>,<at 0.5975 0.4 RCIM_Cubic>]}
      }})
    R1 Pink:LinearColor=&ColorFromCurve(ColorCurve=DataInterface() {propertyValues {
       AlphaCurve {keys [<at 0 1 RCIM_Cubic>,<at 1 1 RCIM_Cubic>]}
       BlueCurve {keys [<at 0.1871 0.9 RCIM_Cubic>,<at 0.5089 0.8188 RCIM_Cubic>]}
       GreenCurve {keys [<at 0.1871 0.45 RCIM_Cubic>,<at 0.5089 0.135 RCIM_Cubic>]}
       RedCurve {keys [<at 0.1871 0.9 RCIM_Cubic>,<at 0.5089 0.9 RCIM_Cubic>]}
      }})
    R1 Purple:LinearColor=&ColorFromCurve(ColorCurve=DataInterface() {propertyValues {
       AlphaCurve {keys [<at 0 1 RCIM_Cubic>,<at 1 1 RCIM_Cubic>]}
       BlueCurve {keys [<at 0.1871 1 RCIM_Cubic>,<at 0.6036 0.4 RCIM_Cubic>]}
       GreenCurve {keys [<at 0.1871 0.43 RCIM_Cubic>,<at 0.6036 0.12 RCIM_Cubic>]}
       RedCurve {keys [<at 0.1871 0.8605 RCIM_Cubic>,<at 0.6036 0.2187 RCIM_Cubic>]}
      }})
    R1 Saffron:LinearColor=&ColorFromCurve(ColorCurve=DataInterface() {propertyValues {
       AlphaCurve {keys [<at 0 1 RCIM_Cubic>,<at 1 1 RCIM_Cubic>]}
       BlueCurve {keys [<at 0.1871 0.1946 RCIM_Cubic>,<at 0.6036 0.1818 RCIM_Cubic>]}
       GreenCurve {keys [<at 0.1871 0.7682 RCIM_Cubic>,<at 0.6036 0.7175 RCIM_Cubic>]}
       RedCurve {keys [<at 0.1871 0.7682 RCIM_Cubic>,<at 0.6036 0.7175 RCIM_Cubic>]}
      }})
    R1 SkyBlue:LinearColor=&ColorFromCurve(ColorCurve=DataInterface() {propertyValues {
       AlphaCurve {keys [<at 0 1 RCIM_Cubic>,<at 1 1 RCIM_Cubic>]}
       BlueCurve {keys [<at 0.1871 1 RCIM_Cubic>,<at 0.6036 0.55 RCIM_Cubic>]}
       GreenCurve {keys [<at 0.1871 0.9762 RCIM_Cubic>,<at 0.6036 0.5253 RCIM_Cubic>]}
       RedCurve {keys [<at 0.1871 0.25 RCIM_Cubic>,<at 0.6036 0.11 RCIM_Cubic>]}
      }})
    R1 TitaniumWhite:LinearColor=&ColorFromCurve(ColorCurve=DataInterface() {propertyValues {
       AlphaCurve {keys [<at 0 1 RCIM_Cubic>,<at 1 1 RCIM_Cubic>]}
       BlueCurve {keys [<at 0.1871 1 RCIM_Cubic>,<at 0.6036 1 RCIM_Cubic>]}
       GreenCurve {keys [<at 0.1871 1 RCIM_Cubic>,<at 0.6036 1 RCIM_Cubic>]}
       RedCurve {keys [<at 0.1871 1 RCIM_Cubic>,<at 0.6036 1 RCIM_Cubic>]}
      }})
    Saffron=&ColorFromCurve(ColorCurve=DataInterface() {propertyValues {
       AlphaCurve {keys [<at 0 1 RCIM_Cubic>,<at 1 1 RCIM_Cubic>]}
       BlueCurve {keys [<at 0.1814 0 RCIM_Cubic>,<at 0.6008 0.0041 RCIM_Cubic>]}
       GreenCurve {keys [<at 0.1814 0.4072 RCIM_Cubic>,<at 0.6008 0.1333 RCIM_Cubic>]}
       RedCurve {keys [<at 0.1814 0.5972 RCIM_Cubic>,<at 0.6008 0.2 RCIM_Cubic>]}
      }})
    SkyBlue=&ColorFromCurve(ColorCurve=DataInterface() {propertyValues {
       AlphaCurve {keys [<at 0 1 RCIM_Cubic>,<at 1 1 RCIM_Cubic>]}
       BlueCurve {keys [<at 0.1814 0.5614 RCIM_Cubic>,<at 0.6008 0.2502 RCIM_Cubic>]}
       GreenCurve {keys [<at 0.1814 0.4615 RCIM_Cubic>,<at 0.6008 0.1812 RCIM_Cubic>]}
       RedCurve {keys [<at 0.1814 0.1286 RCIM_Cubic>,<at 0.6008 0.0356 RCIM_Cubic>]}
      }})
    TitaniumWhite=&ColorFromCurve(ColorCurve=DataInterface() {propertyValues {
       AlphaCurve {keys [<at 0 1 RCIM_Cubic>,<at 1 1 RCIM_Cubic>]}
       BlueCurve {keys [<at 0.1871 0.25 RCIM_Cubic>,<at 0.6036 0.13 RCIM_Cubic>]}
       GreenCurve {keys [<at 0.1871 0.25 RCIM_Cubic>,<at 0.6036 0.1282 RCIM_Cubic>]}
       RedCurve {keys [<at 0.1871 0.25 RCIM_Cubic>,<at 0.6036 0.1282 RCIM_Cubic>]}
      }})
    Use R1 Painted Overrides=true
   ) {module=VC_PaintedVariantsOverride001}
   /SetVariables(
    Particles.RL_WidthScaled:NiagaraFloat=&Multiply_Float(
     A=$Particles.RL_Width
     B=&FloatFromCurve(FloatCurve=DataInterface() {propertyValues {Curve {
         defaultValue=1
         keys [<at 0 0.7 RCIM_Cubic>, {interpMode=RCIM_Cubic, tangentMode=RCTM_User, time=0.5, value=1.4}, {tangentMode=RCTM_User, time=1, value=0}]
        }}})
    )
    Particles.RR_WidthScaled:NiagaraFloat=&Multiply_Float(
     A=$Particles.RR_Width
     B=&FloatFromCurve(FloatCurve=DataInterface() {propertyValues {Curve {
         defaultValue=1
         keys [<at 0 0.7 RCIM_Cubic>, {interpMode=RCIM_Cubic, tangentMode=RCTM_User, time=0.5, value=1.4}, {tangentMode=RCTM_User, time=1, value=0}]
        }}})
    )
   ) {isSetParameters=true}
  ]
  rendererClasses [&NiagaraRibbonRendererProperties]
  renderers [
   {rendererClass=&NiagaraRibbonRendererProperties}
   {rendererClass=&NiagaraRibbonRendererProperties, rendererIndex=1}
   {rendererClass=&NiagaraRibbonRendererProperties, rendererIndex=2}
   {rendererClass=&NiagaraRibbonRendererProperties, rendererIndex=3}
  ]
 }
 NE_VC_WheelWrap {
  emitterUpdateScript [
   &EmitterState(
    Inactive Response=::NewEnumerator1
    Life Cycle Mode=::NewEnumerator1
    Loop Behavior=::NewEnumerator1
    Loop Duration Mode=::NewEnumerator1
    UseLoopDelay=false
   )
   &SpawnBurst_Instantaneous(Spawn Count=2, Use Spawn Probability=false, UseLoopCountLimit=false)
  ]
  particleSpawnScript [
   &VC_PaintedVariantsOverride(
    Black=<rgba 0.0278 0.0298 0.035 1>
    BurntSienna=<rgba 0.35 0.0292 0 1>
    Cobalt=<rgba 0.3005 0.7835 1 1>
    Crimson=<rgba 1 0.1398 0.1443 1>
    DefaultColor=<rgba 0.9823 0.9823 0.9823 1>
    ForestGreen=<rgba 0.4262 0.65 0.281 1>
    Grey=<rgba 0.7 0.7 0.7 1>
    Lime=<rgba 1 1 0.3 1>
    Orange=<rgba 0.7011 0.624 0.0704 1>
    Pink=<rgba 0.9 0.495 0.9 1>
    Purple=<rgba 0.8754 0.48 1 1>
    R1 Black:LinearColor=<rgba 0.0119 0.0128 0.015 1>
    R1 BurntSienna:LinearColor=<rgba 0.1 0.012 0.0039 1>
    R1 Cobalt:LinearColor=<rgba 0.0124 0.0492 0.45 1>
    R1 Crimson:LinearColor=<rgba 0.35 0.009 0.0097 1>
    R1 DefaultColor=<rgba 0.3608 0.35 0.0634 1>
    R1 ForestGreen:LinearColor=<rgba 0.0546 0.2 0.0336 1>
    R1 Grey:LinearColor=<rgba 0.1938 0.1938 0.2 1>
    R1 Lime:LinearColor=<rgba 0.1533 0.25 0.0243 1>
    R1 Orange:LinearColor=<rgba 0.35 0.048 0.0072 1>
    R1 Pink:LinearColor=<rgba 0.21 0.0206 0.1433 1>
    R1 Purple:LinearColor=<rgba 0.0699 0.0403 0.4 1>
    R1 Saffron:LinearColor=<rgba 0.202 0.1332 0.004 1>
    R1 SkyBlue:LinearColor=<rgba 0.0285 0.1448 0.2 1>
    R1 TitaniumWhite:LinearColor=<rgba 0.1959 0.1959 0.2 1>
    R2 Black:LinearColor=<rgba 0.0095 0.0102 0.012 1>
    R2 BurntSienna:LinearColor=<rgba 0.1 0.012 0.0039 1>
    R2 Cobalt:LinearColor=<rgba 0 0.118 0.8 1>
    R2 Crimson:LinearColor=<rgba 0.75 0.0266 0.0278 1>
    R2 DefaultColor:LinearColor=<rgba 1 0.1 0 1>
    R2 ForestGreen:LinearColor=<rgba 0.1335 0.55 0.0541 1>
    R2 Grey:LinearColor=<rgba 0.35 0.35 0.35 1>
    R2 Lime:LinearColor=<rgba 0.4237 0.65 0 1>
    R2 Orange:LinearColor=<rgba 0.65 0.1322 0.0142 1>
    R2 Pink:LinearColor=<rgba 0.95 0.19 0.8389 1>
    R2 Purple:LinearColor=<rgba 0.2347 0.1406 0.7 1>
    R2 Saffron:LinearColor=<rgba 0.5972 0.4072 0 1>
    R2 SkyBlue:LinearColor=<rgba 0.1354 0.491 0.6 1>
    R2 TitaniumWhite:LinearColor=<rgba 0.3 0.3 0.3 1>
    R3 Black:LinearColor=<rgba 0.0278 0.0298 0.035 1>
    R3 BurntSienna:LinearColor=<rgba 0.35 0.0292 0 1>
    R3 Cobalt:LinearColor=<rgba 0.1653 0.4309 0.55 1>
    R3 Crimson:LinearColor=<rgba 0.7 0.0979 0.101 1>
    R3 DefaultColor:LinearColor=<rgba 0.682 0.682 0.682 1>
    R3 ForestGreen:LinearColor=<rgba 0.3278 0.5 0.2162 1>
    R3 Grey:LinearColor=<rgba 0.4 0.4 0.4 1>
    R3 Lime:LinearColor=<rgba 0.55 0.55 0.165 1>
    R3 Orange:LinearColor=<rgba 0.4 0.356 0.0401 1>
    R3 Pink:LinearColor=<rgba 0.5 0.25 0.5 1>
    R3 Purple:LinearColor=<rgba 0.4377 0.24 0.5 1>
    R3 Saffron:LinearColor=<rgba 0.468 0.468 0.1638 1>
    R3 SkyBlue:LinearColor=<rgba 0.225 0.4903 0.5 1>
    R3 TitaniumWhite:LinearColor=<rgba 0.5 0.5 0.5 1>
    Saffron=<rgba 0.7682 0.7682 0.2689 1>
    SkyBlue=<rgba 0.45 0.9805 1 1>
    TitaniumWhite=<rgba 0.9 0.9 0.9 1>
    Use R1 Painted Overrides=true
    Use R2 Painted Overrides=true
    Use R3 Painted Overrides=true
   )
   &InitializeParticle(
    Color Mode=::NewEnumerator0
    Lifetime=$Emitter.CurrentLoopDuration
    Mesh Scale Mode=::NewEnumerator4
    Sprite Rotation Mode=::NewEnumerator2
    Sprite Size Mode=::NewEnumerator3
    UsePositionOffset=false
    Write Material Random=true
   )
   &PartitionParticles()
   &SkeletalMeshLocation(
    Bone / Socket Index=&Add_Integer(A=$Particles.PartitionParticles.Partition, B=4)
    Bone / Socket Sampling Mode=::NewEnumerator5
    Interpolate Mesh Sampling=false
    MaskSpawnGroup=false
    Mesh Sampling Type=::NewEnumerator6
    Skeletal Mesh=$System.SkeletalMesh
   )
  ]
  particleUpdateScript [
   &FN_DisableTireFX(
    Mask Index Attribute:NiagaraInt32=$Particles.PartitionParticles.Partition
    Mask Index FL=2
    Mask Index FR=3
    Mask Index RL=0
    Mask Index RR=1
   )
   &VC_WheelWrapScale(Wheel Index Attribute:NiagaraInt32=&Add_Integer(A=$Particles.PartitionParticles.Partition, B=2))
   &ParticleState()
   &UpdateMeshOrientation(Rotation Coordinate Space=::NewEnumerator2, Rotation Vector=<0 1 0>)
   &SkeletalMeshLocation(
    Bone / Socket Index=&Add_Integer(A=$Particles.PartitionParticles.Partition, B=4)
    Bone / Socket Sampling Mode=::NewEnumerator5
    MaskSpawnGroup=false
    Mesh Sampling Type=::NewEnumerator6
    Skeletal Mesh=$System.SkeletalMesh
   ) {module=SkeletalMeshLocation001}
   &SolveForcesAndVelocity(Clamp Velocity=false, Limit Acceleration=false)
  ]
  rendererClasses [&NiagaraMeshRendererProperties]
  renderers [
   {rendererClass=&NiagaraMeshRendererProperties, rendererIndex=0}
   {rendererClass=&NiagaraMeshRendererProperties, rendererIndex=1}
   {rendererClass=&NiagaraMeshRendererProperties, rendererIndex=2}
   {rendererClass=&NiagaraMeshRendererProperties, rendererIndex=3}
   {rendererClass=&NiagaraMeshRendererProperties, rendererIndex=4}
  ]
 }
}
summary {
 systemName=NS_DriftTrail_Psychomatic_MP
 emitters [
  {emitterName=NE_VC_SubUV_Sprite, rendererClasses [&NiagaraSpriteRendererProperties]}
  {emitterName=NE_VC_Trail_Ribbon, rendererClasses [&NiagaraRibbonRendererProperties]}
  {emitterName=NE_VC_WheelWrap, rendererClasses [&NiagaraMeshRendererProperties]}
 ]
 userVariables [
  User.bIsFrontendPreview {type=NiagaraBool}
  User.bIsPainted {type=NiagaraBool}
  User.ColorVariantID {type=NiagaraInt32}
  User.GameModeID {type=NiagaraInt32}
  User.offset {type=NiagaraPosition, description="", defaultValue {}}
  User.PaintedColor {type=LinearColor, description="", defaultValue {struct=LinearColor, value=<rgba 1 1 1 1>}}
  User.PaintedScalar {type=NiagaraFloat, defaultValue=1, description=""}
  User.VehicleFlags {type=NiagaraInt32, defaultValue=806289408}
  User.Wheel_Radius_Array {
   type=NiagaraDataInterfaceArrayFloat
   defaultValue {
    struct=NiagaraExt_VariableValue_DataInterface
    value {
     dataInterface=&self:NiagaraDataInterfaceArrayFloat_1
     dataInterfaceClass=&NiagaraDataInterfaceArrayFloat
    }
   }
  }
  User.Wheel_Width_Array {
   type=NiagaraDataInterfaceArrayFloat
   defaultValue {
    struct=NiagaraExt_VariableValue_DataInterface
    value {
     dataInterface=&self:NiagaraDataInterfaceArrayFloat_0
     dataInterfaceClass=&NiagaraDataInterfaceArrayFloat
    }
   }
  }
  User.WheelTouchingArray {
   type=NiagaraDataInterfaceArrayFloat
   defaultValue {
    struct=NiagaraExt_VariableValue_DataInterface
    value {
     dataInterface=&self:NiagaraDataInterfaceArrayFloat_2
     dataInterfaceClass=&NiagaraDataInterfaceArrayFloat
    }
   }
  }
 ]
}
